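# Niagara idle effect for the Sparks Outershell microphone stand instrument cosmetic, showing a persistent blue capsule glow.
# A single CapsuleGlow emitter burst-spawns long-lived particles at a fixed offset, pulled toward the camera, with sprite and mesh renderers.
# Sprite size follows owner scale, and dynamic material parameters feed the glow material.
# Color is multiplied by a system time-of-day color, with alpha switched between gameplay and a dimmer front-end value.
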NiagaraSystem /SparksCosmetics/Instrument/MicStand/Outershell/Effects/Mic_Instrument_Idle/NS_Mic_Outershell_Idle_FX
dependencies {
 usedDataInterfaces [&NiagaraDataInterfaceCurve]
 usedDynamicInputs [
  &MergeHelper_Vector2Switch
  &FNTimeOfDayColor
  &MakeFloatFromVector
 ]
 usedModules [
  &SystemState
  &self:SystemSpawnScript.SystemScriptSource.SystemScriptGraph.NiagaraNodeAssignment_10.SetVariables_#1
  &self:SystemSpawnScript.SystemScriptSource.SystemScriptGraph.NiagaraNodeAssignment_2.SetVariables_#2
 ]
}
emitters {CapsuleGlow {
  emitterUpdateScript [
   &EmitterState(Life Cycle Mode=::NewEnumerator1, Loop Behavior=::NewEnumerator1, Loop Duration Mode=::NewEnumerator1, UseLoopDelay=false)
   &SpawnBurst_Instantaneous(Use Spawn Probability=false, UseLoopCountLimit=true)
  ]
  particleSpawnScript [
   &PartitionParticles(ID to Partition:NiagaraInt32=&ReturnExecIndex(), Number Of Partitions=5) {enabled=false}
   &InitializeParticle(
    Color=<rgba 0 0.5362 1 0.25>
    Position=<-1.3 0 3.8>
    Position Mode=::NewEnumerator1
    Sprite Size Mode=::NewEnumerator3
    Uniform Sprite Size=11
    UsePositionOffset=false
    Write Material Random=true
   )
   &CameraOffset(Camera Offset Amount=3.5)
  ]
  particleUpdateScript [
   &ParticleState(Kill Particles When Lifetime Has Elapsed=false, Let Infinitely Lived Particles Die When Emitter Deactivates=true)
   &ScaleSpriteSize(Scale Sprite Size Mode=::NewEnumerator0, Uniform Scale Factor=$System.EngineOwnerScale)
   &DynamicMaterialParameters(
    Index 0 Param 1=1
    Index 0 Param 2=5
    Module.Index0Param0Write=true
    Module.Index0Param1Write=true
    Module.Index0Param2Write=false
    Module.Index0Param3Write=false
   ) {module=DynamicMaterialParameters001}
   &ScaleColor(
    Color Scale:LinearColor=&MultiplyLinearColor(
     Linear Color=$System.TODColor
     Scale Factor=&MakeVector4FromVectorAndFloat("Vector (RGB)"=<1 1 1>, "Float (Alpha)"=&FN_FrontEndFloatModifier(Gameplay=1, Front End=0.6))
    )
    Scale Mode=3
   ) {module=ScaleColor001}
  ]
  rendererClasses [&NiagaraSpriteRendererProperties, &NiagaraMeshRendererProperties]
  renderers [
   {rendererClass=&NiagaraSpriteRendererProperties}
   {rendererClass=&NiagaraMeshRendererProperties, rendererIndex=1}
  ]
 }}
summary {
 systemName=NS_Mic_Outershell_Idle_FX
 emitters [{emitterName=CapsuleGlow, rendererClasses [&NiagaraSpriteRendererProperties, &NiagaraMeshRendererProperties]}]
 userVariables [
  User.bIsFrontEnd {type=NiagaraBool}
  User.bIsFrontEndPreview {type=NiagaraBool}
 ]
}
